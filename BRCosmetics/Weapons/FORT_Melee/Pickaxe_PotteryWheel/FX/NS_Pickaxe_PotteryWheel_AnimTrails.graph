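# Niagara system for the swing trails of the Pottery Wheel pickaxe, drawn as ribbons that follow bones or sockets on a skeletal mesh.
# Four emitters share the FN AnimTrail spawn, initialize and update modules: soft dark, soft red, thin white tip and cylinder trails.
# Short-lived particles get a user velocity and curve-driven color, with time-of-day tinting on most trails.
# The cylinder trail uses two ribbon renderers and also writes inner-light position, color and width.
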NiagaraSystem /BRCosmetics/Weapons/FORT_Melee/Pickaxe_PotteryWheel/FX/NS_Pickaxe_PotteryWheel_AnimTrails
dependencies {
 usedDataInterfaces [&NiagaraDataInterfaceColorCurve, &NiagaraDataInterfaceCurve, &NiagaraDataInterfaceSkeletalMesh]
 usedDynamicInputs []
 usedModules [
  &self:SystemSpawnScript.SystemScriptSource.SystemScriptGraph.NiagaraNodeAssignment_0.SetVariables_#1
  &SystemState
 ]
}
emitters {
 NE_CylinderTrail {
  emitterUpdateScript [
   &EmitterState()
   &FN_SpawnAnimTrail(AnimTrailWidthMode=::FromSecond)
  ]
  particleSpawnScript [
   &FN_InitializeAnimTrailBoneIndecies(Second Bone or Socket Index=2)
   &FN_InitializeAnimTrail(
    AnimTrailWidth=0.3
    Color=&FN_Select_Color_from_Time_Of_Day(Morning=<rgba 1 1 1 1>, Day=<rgba 1 1 1 1>, Evening=<rgba 1 1 1 1>, Night=<rgba 1 1 1 1>)
    Lifetime=0.075
    SkeletalMesh=$System.SkeletalMesh
   )
   /SetVariables(Particles.Position=&AddVectorToPosition(Position=$Particles.Position, Vector=&Multiply_VectorByFloat(Vector=$Particles.RibbonFacing, Float=0))) {isSetParameters=true}
   /SetVariables(Particles.InnerLightPos:NiagaraPosition=&AddVectorToPosition(Position=$Particles.Position, Vector=&Multiply_VectorByFloat(Vector=$Particles.RibbonFacing, Float=0))) {isSetParameters=true}
   &AddVelocity(Velocity=$User.NewVector3f)
  ]
  particleUpdateScript [
   &SolveForcesAndVelocity(Clamp Velocity=false, Limit Acceleration=false)
   &FN_UpdateAnimTrail(Locked Particle RibbonWidth=1, SkeletalMesh=$System.SkeletalMesh) {module=FN_UpdateAnimTrail001}
   &ScaleColor(
    Scale Alpha=&FloatFromCurve(
     FloatCurve=DataInterface() {propertyValues {Curve {keys [
         {arriveTangent=-2.003, interpMode=RCIM_Cubic, leaveTangent=-2.003, tangentMode=RCTM_User, time=0.5, value=0.95}
         {interpMode=RCIM_Cubic, tangentMode=RCTM_User, time=1, value=0}
        ]}}}
     Scale Curve=0.1
    )
    Scale RGB=&VectorFromFloat(Value=&FNTimeOfDayScalar(Night Scale=1, Morning Scale=1, Evening Scale=1))
    ScaleA=true
    ScaleRGB=true
   )
   &Color(Color=&MultiplyLinearColor(
     Linear Color=&ColorFromCurve(ColorCurve=DataInterface() {propertyValues {
        AlphaCurve {keys [<at 0 1 RCIM_Cubic>]}
        BlueCurve {keys [<at 0 0.2917 RCIM_Cubic>]}
        GreenCurve {keys [<at 0 0.2917 RCIM_Cubic>]}
        RedCurve {keys [<at 0 0.2917 RCIM_Cubic>]}
       }})
     Scale Factor=&MakeVector4_FromLinearColor(LinearColor=&FNTimeOfDayColor(
       Night Color=<rgba 0.3021 0.3592 0.6042 1>
       Morning Color=<rgba 1 0.5204 0.5403 1>
       Day Color=<rgba 1 1 1 1>
       Evening Color=<rgba 1 0.3293 0.3073 1>
      ))
    ))
   /SetVariables(Particles.InnerlightColor:LinearColor=&MultiplyLinearColor(
     Linear Color=&ColorFromCurve(ColorCurve=DataInterface() {propertyValues {
        AlphaCurve {keys [<at 0 1 RCIM_Cubic>,<at 0.3814 0>]}
        BlueCurve {keys [<at 0 0 RCIM_Cubic>]}
        GreenCurve {keys [<at 0 0.1 RCIM_Cubic>]}
        RedCurve {keys [<at 0 1 RCIM_Cubic>]}
       }})
     Scale Factor=&MakeVector4_FromLinearColor(LinearColor=&FNTimeOfDayColor(Night Color=<rgba 0.2092 0.2215 0.375 1>, Morning Color=<rgba 1 1 1 1>, Day Color=<rgba 1 1 1 1>, Evening Color=<rgba 1 0.4062 0.4062 1>))
    )) {isSetParameters=true}
   /SetVariables(Particles.InnerlightWidth:NiagaraFloat=&Multiply_Float(A=$Particles.RibbonWidth, B=0.5)) {isSetParameters=true}
  ]
  rendererClasses [&NiagaraRibbonRendererProperties]
  renderers [
   {rendererClass=&NiagaraRibbonRendererProperties}
   {rendererClass=&NiagaraRibbonRendererProperties, rendererIndex=1}
  ]
 }
 NE_SoftDarkTrail {
  emitterUpdateScript [
   &EmitterState()
   &FN_SpawnAnimTrail()
  ]
  particleSpawnScript [
   &FN_InitializeAnimTrailBoneIndecies()
   &FN_InitializeAnimTrail(
    Color=&FN_Select_Color_from_Time_Of_Day(Morning=<rgba 0 2.8891 4 1>, Day=<rgba 0 3.8853 5 1>, Evening=<rgba 0 3.5105 4 1>, Night=<rgba 0 3.8978 5 1>)
    Lifetime=0.1
    SkeletalMesh=$System.SkeletalMesh
   )
   &AddVelocity(Velocity=$User.NewVector3f)
  ]
  particleUpdateScript [
   &FN_UpdateAnimTrail(SkeletalMesh=$System.SkeletalMesh) {module=FN_UpdateAnimTrail001}
   &Color(Color=&ColorFromCurve(ColorCurve=DataInterface() {propertyValues {
       AlphaCurve {keys [<at 0 1 RCIM_Cubic>,<at 0.4607 0.25>,<at 1 0 RCIM_Cubic>]}
       BlueCurve {keys [<at 0 0.0156>]}
       GreenCurve {keys [<at 0 0.0156>]}
       RedCurve {keys [<at 0 0.0156>]}
      }}))
   &SolveForcesAndVelocity(Clamp Velocity=false, Limit Acceleration=false)
  ]
  rendererClasses [&NiagaraRibbonRendererProperties]
  renderers [{rendererClass=&NiagaraRibbonRendererProperties}]
 }
 NE_SoftRedTrail {
  emitterUpdateScript [
   &EmitterState()
   &FN_SpawnAnimTrail()
  ]
  particleSpawnScript [
   &FN_InitializeAnimTrailBoneIndecies(Second Bone or Socket Index=3)
   &FN_InitializeAnimTrail(
    Color=&FN_Select_Color_from_Time_Of_Day(Morning=<rgba 1 1 1 1>, Day=<rgba 1 1 1 1>, Evening=<rgba 1 1 1 1>, Night=<rgba 1 1 1 1>)
    Lifetime=0.1
    SkeletalMesh=$System.SkeletalMesh
   )
   &AddVelocity(Velocity=$User.NewVector3f)
  ]
  particleUpdateScript [
   &FN_UpdateAnimTrail(SkeletalMesh=$System.SkeletalMesh) {module=FN_UpdateAnimTrail001}
   &Color(Color=&ColorFromCurve(ColorCurve=DataInterface() {propertyValues {
       AlphaCurve {keys [<at 0 0.8>,<at 0.1447 0.2>,<at 0.6442 0 RCIM_Cubic>]}
       BlueCurve {keys [<at 0 0>]}
       GreenCurve {keys [<at 0 0.06>]}
       RedCurve {keys [<at 0 1>]}
      }}))
   /SetVariables(Particles.Color=&MultiplyLinearColor(
     Linear Color=&MultiplyLinearColor(
      Linear Color=$Particles.Color
      Scale Factor=&MakeVector4_FromLinearColor(LinearColor=&FNTimeOfDayColor(Night Color=<rgba 0.5 0.5 0.5 1>, Morning Color=<rgba 1 1 1 1>, Day Color=<rgba 1 1 1 1>, Evening Color=<rgba 1 1 1 1>))
     )
     Scale Factor=&MakeVector4(X=1, Y=1, Z=1, W=1)
    )) {isSetParameters=true}
   &SolveForcesAndVelocity(Clamp Velocity=false, Limit Acceleration=false)
  ]
  rendererClasses [&NiagaraRibbonRendererProperties]
  renderers [{rendererClass=&NiagaraRibbonRendererProperties}]
 }
 NE_TipWhiteTrail {
  emitterUpdateScript [
   &EmitterState()
   &FN_SpawnAnimTrail(AnimTrailWidthMode=::FromSecond)
  ]
  particleSpawnScript [
   &FN_InitializeAnimTrailBoneIndecies()
   &FN_InitializeAnimTrail(AnimTrailWidth=0.05, Lifetime=0.1, SkeletalMesh=$System.SkeletalMesh)
   &AddVelocity(Velocity=$User.NewVector3f)
  ]
  particleUpdateScript [
   &FN_UpdateAnimTrail(SkeletalMesh=$System.SkeletalMesh) {module=FN_UpdateAnimTrail001}
   &Color(Color=&MultiplyLinearColor(
     Linear Color=&ColorFromCurve(ColorCurve=DataInterface() {propertyValues {AlphaCurve {keys [<at 0 1>,<at 0.345 0 RCIM_Cubic>]}, BlueCurve {keys [<at 0 1>]}, GreenCurve {keys [<at 0 1>]}, RedCurve {keys [<at 0 1>]}}})
     Scale Factor=&MakeVector4_FromLinearColor(LinearColor=&FNTimeOfDayColor(Night Color=<rgba 0.3021 0.3592 0.6042 1>, Morning Color=<rgba 1 1 1 1>, Day Color=<rgba 1 1 1 1>, Evening Color=<rgba 1 0.789 0.6133 1>))
    ))
   &SolveForcesAndVelocity(Clamp Velocity=false, Limit Acceleration=false)
  ]
  rendererClasses [&NiagaraRibbonRendererProperties]
  renderers [{rendererClass=&NiagaraRibbonRendererProperties}]
 }
}
summary {
 systemName=NS_Pickaxe_PotteryWheel_AnimTrails
 emitters [
  {emitterName=NE_SoftDarkTrail, rendererClasses [&NiagaraRibbonRendererProperties]}
  {emitterName=NE_SoftRedTrail, rendererClasses [&NiagaraRibbonRendererProperties]}
  {emitterName=NE_TipWhiteTrail, rendererClasses [&NiagaraRibbonRendererProperties]}
  {emitterName=NE_CylinderTrail, rendererClasses [&NiagaraRibbonRendererProperties]}
 ]
 userVariables [
  User.NewVector3f {type=Vector3f, description="", defaultValue {struct=Vector3f, value=<0 0 0>}}
  User.RibbonIDOffset {type=NiagaraInt32}
 ]
}
