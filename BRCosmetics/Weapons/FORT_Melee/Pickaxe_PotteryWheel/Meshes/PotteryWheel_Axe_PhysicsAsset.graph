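# Physics asset for the Pottery Wheel pickaxe, a two-handed melee weapon cosmetic.
# It has a single capsule body on the Melee_TwoHanded_Handle bone, offset along the handle and rotated to fit the weapon mesh.
# There are no constraints, so the weapon collides as one rigid shape with no articulated parts.
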PhysicsAsset /BRCosmetics/Weapons/FORT_Melee/Pickaxe_PotteryWheel/Meshes/PotteryWheel_Axe_PhysicsAsset
bodies {Melee_TwoHanded_Handle [{shapeType=Capsule, length=58.6867, radius=26.6393, center=<0.7936 -1.1605 27.354>, rotation=<rot 0.0271 -90 0.0082>}]}
constraints []
